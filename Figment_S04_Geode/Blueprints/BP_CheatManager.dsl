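; BP_CheatManager is a debug Blueprint that loads and unloads Geode event content through World Partition data layers via console cheat commands.
; Four custom events, noted as "Ke *" commands, cover load and unload for the Geode event and Geode warmup layer sets.
; Each loops over its data layer array, skips invalid entries, sets the runtime state to Activated or Unloaded non-recursively, and prints a confirmation per layer.
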(event Custom|LoadGeode
  ; Cheat Ke* LoadGeode
  (for _array_element (Variables|Default|GetGeodeEventDataLayers)
    (Utilities|IsValid _array_element
      (:Is%20Valid
        (bind _returnvalue (WorldPartition|GetDataLayerManager))
        (DataLayers|SetDataLayerRuntimeState _returnvalue _array_element "Activated" false :!class /Script/Engine.DataLayerManager)
        (Development|PrintString "Geode Activated")))))

(event Custom|UnloadGeode
  ; Cheat Ke* UnloadGeode
  (for _array_element (Variables|Default|GetGeodeEventDataLayers)
    (Utilities|IsValid _array_element
      (:Is%20Valid
        (bind _returnvalue (WorldPartition|GetDataLayerManager))
        (DataLayers|SetDataLayerRuntimeState _returnvalue _array_element "Unloaded" false :!class /Script/Engine.DataLayerManager)
        (Development|PrintString "Geode Unloaded")))))

(event Custom|LoadGeodeWarmup
  ; Cheat Ke * LoadGeodeWarmup
  (for _array_element (Variables|Default|GetGeodeWarmupDataLayers)
    (Utilities|IsValid _array_element
      (:Is%20Valid
        (bind _returnvalue (WorldPartition|GetDataLayerManager))
        (DataLayers|SetDataLayerRuntimeState _returnvalue _array_element "Activated" false :!class /Script/Engine.DataLayerManager)
        (Development|PrintString "Geode Warmup Activated")))))

(event Custom|UnloadGeodeWarmup
  ; Cheat ke * UnloadGeodeWarmup
  (for _array_element (Variables|Default|GetGeodeWarmupDataLayers)
    (Utilities|IsValid _array_element
      (:Is%20Valid
        (bind _returnvalue (WorldPartition|GetDataLayerManager))
        (DataLayers|SetDataLayerRuntimeState _returnvalue _array_element "Unloaded" false :!class /Script/Engine.DataLayerManager)
        (Development|PrintString "Geode Warmup Unloaded")))))

(fn ConstructionScript)
