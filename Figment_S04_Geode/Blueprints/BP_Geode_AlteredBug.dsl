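; Geode event actor that plays the AlteredBug gameplay cue on players shortly before the Geode calendar event ends.
; At match start, the authority checks playlist and calendar, then delays to the trigger time if under an hour remains.
; ExecuteAlteredBug applies the local cue only on remotes, removing it after 30 seconds.
; Targets fall back from pawn to view target; the spectator fallback sets TargetPawn, but TargetActor is returned.
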(event EventBeginPlay
  (|CallParentFunction)
  (bind _asfort_game_mode_zone (Utilities|Casting|CastToFortGameModeZone (Game|GetGameMode)))
  (EventDispatchers|BindEventtoOnHandleMatchHasStarted _asfort_game_mode_zone (EventDispatchers|CreateEvent :!event Custom|OnMatchHasStarted)))

(event Custom|OnMatchHasStarted
  (bind _triggertimebeforeendofevent (Variables|Default|GetTriggerTimeBeforeEndOfEvent))
  (Networking|SwitchHasAuthority
    (:Authority
      (Utilities|WatchShouldRunEventinPlaylist
        (:ShouldRun
          (MCPContext|InitialCalendarSyncCompleted
            (:OnSyncComplete
              (bind (_biseventactive _timesincebegin _timeuntilend) (CalendarEvents|GetCalendarEventInformation (Variables|Default|GetGeodeBeginCalendarEvent)))
              (if _biseventactive
                ; less than 1 hour
                (if (< (Math|Timespan|GetTotalHours _timeuntilend) 1.0)
                  (Development|PrintString (Utilities|String|ToString(Text) (Utilities|Text|FormatText "NSLOCTEXT(\"[58CEC509BA64CA3F24577C35CCBAD966]\", \"A077E78949D46A31C398238CF5A5964B\", \"waiting for {t}s to activate AlteredBug\")" (- (Math|Timespan|GetTotalSeconds _timeuntilend) _triggertimebeforeendofevent))))
                  (Utilities|FlowControl|Delay (- (Math|Timespan|GetTotalSeconds _timeuntilend) _triggertimebeforeendofevent))
                  (Development|PrintString "Trigerring AlteredBug")
                  (CallFunction|ExecuteAlteredBug))))))))))

(event Custom|TestGeodeAlteredBug
  ; Cheat ke * TestGeodeAlteredBug
  (CallFunction|ExecuteAlteredBug))

(event Custom|ExecuteAlteredBug
  (Networking|SwitchHasAuthority
    (:Remote
      (for _index (range (Game|GetNumPlayerControllers))
        (CallFunction|ExecuteALteredBugonActor (CallFunction|GetTargetActorfromPlayerController (Game|GetPlayerController _index)))))))

(event Custom|ExecuteALteredBugOnActor
  (params (TargetActor :type Actor%20Object%20Reference))
  (Utilities|IsValid TargetActor
    (:Is%20Valid
      (Projectile|AddGameplayCue%28Looping%29Local TargetActor (Variables|Default|GetAlteredBugCueTag) (Ability|GameplayCue|MakeGameplayCueParameters :NormalizedMagnitude 1.0 :RawMagnitude 1e+01 :OriginalTag (Variables|Default|GetAlteredBugCueTag) :Location "0,0,0" :Normal "0,0,0" :EffectCauser self :SourceObject (Ability|GetAbilitySystemComponent TargetActor) :GameplayEffectLevel 1 :AbilityLevel 1 :bReplicateLocationWhenUsingMinimalRepProxy false) false)
      (Utilities|FlowControl|Delay 3e+01)
      (GameplayCue|RemoveGameplayCue TargetActor (Variables|Default|GetAlteredBugCueTag)))))

(fn ExecuteAlteredBugOnPlayerController
  (params (self2 :type Controller%20Object%20Reference))
  (Projectile|AddGameplayCue%28Looping%29Local (Pawn|GetControlledPawn self2) (Variables|Default|GetAlteredBugCueTag) (Ability|GameplayCue|MakeGameplayCueParameters :NormalizedMagnitude 1.0 :RawMagnitude 1e+01 :OriginalTag (Variables|Default|GetAlteredBugCueTag) :Location "0,0,0" :Normal "0,0,0" :EffectCauser self :SourceObject (Ability|GetAbilitySystemComponent (Pawn|GetControlledPawn self2)) :GameplayEffectLevel 1 :AbilityLevel 1 :bReplicateLocationWhenUsingMinimalRepProxy false) false))

(fn GetTargetActorFromPlayerController
  (params (PlayerController :type Controller%20Object%20Reference))
  (Utilities|FlowControl|Sequence
    (:then_0
      (bind _asfort_player_controller (Utilities|Casting|CastToFortPlayerController PlayerController))
      ; Try Getting Controlled pawn
      (bind _output_get (LocalVariables|SetTargetActor (Pawn|GetControlledPawn _asfort_player_controller)))
      ; Try getting view target
      (Utilities|IsValid _output_get
        (:Is%20Not%20Valid
          (bind _output_get_2 (LocalVariables|SetTargetActor (Pawn|GetViewTarget _asfort_player_controller)))
          ; Try getting Spactator Target
          (Utilities|IsValid _output_get_2
            (:Is%20Not%20Valid
              (LocalVariables|SetTargetPawn (Pawn|GetSpectatorPawn _asfort_player_controller)))))))
    (:then_1
      (return (LocalVariables|GetTargetActor)))))

(fn ConstructionScript)
